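; Parent Blueprint for Creative decal devices, showing a decal whose color follows knob and team-color settings.
; On device init it async-loads the decal material into a cached dynamic instance, if none exists, and seeds the CreativeColorComponent.
; Client color updates write the adjusted HSVA color, converted to RGBA, into that instance's ColorLight parameter.
; A handle mesh supplies collision and in-game visibility control, and the material-override hook reapplies the decal instance.
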(event EventInitializeDevice
  (|CallParentFunction)
  ;[0.5 1 1] Getting the decal MID
  (Variables|Default|GetLoadedDecalMID
    (:then
      _)
    (:else
      (bind _object (Utilities|AsyncLoadAsset (Variables|Default|GetDecalMaterial))
        (:Completed
          (bind _asmaterial_interface (Utilities|Casting|CastToMaterialInterface _object))
          (bind _returnvalue (Rendering|Material|CreateDynamicMaterialInstance _asmaterial_interface "None" "None" :!class /Script/Engine.KismetMaterialLibrary))
          (bind _output_get (Variables|Default|SetLoadedDecalMID _returnvalue))
          (Rendering|Components|Decal|SetDecalMaterial (Variables|Default|GetDecal) _output_get)))))
  ;[1 1 0.5] Initial Values
  (Class|CreativeColorComponent|SetInitialValues (Variables|Default|GetCreativeColorComponent) (Variables|Knobs|GetKnob_InitialTeamColorFilter) (Variables|Knobs|GetKnob_Color) "(R=1.000000,G=1.000000,B=1.000000,A=1.000000)" 1.0 2.0 0))

(event UpdateVisualsonClient(CreativeColorComponent)
  ;[1 0.5 1] Updating ColorLight on the MID
  (bind _loadeddecalmid (Variables|Default|GetLoadedDecalMID)
    (:then
      (Rendering|Material|SetVectorParameterValue _loadeddecalmid "ColorLight" (Class|CreativeColorComponentBlueprintLibrary|HSVAToRGBA (Class|CreativeColorComponent|GetAdjustedColor (Variables|Default|GetCreativeColorComponent))) :!class /Script/Engine.MaterialInstanceDynamic))
    (:else)))

(event EventOnWorldReady
  (|CallParentFunction)
  (Visibility|SetVisibilityComponents (Variables|Default|GetVisibleInGameComponent) (CallFunction|GetVisibilityComponents) true))

(fn GetVisibilityComponents
  (return (Utilities|Array|MakeArray (Variables|Default|GetHandle))))

(fn GetCollisionStaticMesh
  (return "0, 0, 0" (Variables|/Script/Engine.StaticMeshComponent|GetStaticMesh (Variables|Default|GetHandle))))

(fn GetMeshComponents
  (return 0))

(fn GetOverrideMeshMaterial
  (params (MeshComp :type Static%20Mesh%20Component%20Object%20Reference) (MatIdx :type Integer))
  (Rendering|Components|Decal|SetDecalMaterial (Variables|Default|GetDecal) (Variables|Default|GetLoadedDecalMID))
  (return 0))

(fn IgnoreWorldGridScaleOverride
  (return false))

(fn ConstructionScript)
